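# Niagara system for the LemonWrath glider's trail and speed-line effects, with separate tuning for gameplay and front-end preview.
# MainTrail spawns two anim trails from bone or socket pairs, drawn by three ribbon renderers and colored by variant index and time of day.
# SpeedLine spawns short-lived GPU sprites stretched by velocity.
# System-level InterpolateOverTime values blend trail velocity and width and fade speed lines in gameplay; preview speed lines require full deployment.
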NiagaraSystem /BRCosmetics/Gadgets/Assets/VinderTech_GliderChute/Glider_LemonWrath/FX/NS_Glider_LemonWrath
dependencies {
 usedDataInterfaces [
  &NiagaraDataInterfaceVector2DCurve
  &NiagaraDataInterfaceCurve
  &NiagaraDataInterfaceArrayColor
  &NiagaraDataInterfaceArrayFloat3
  &NiagaraDataInterfaceSocketReader
 ]
 usedDynamicInputs [&SetBoolByFloatComparison]
 usedModules [
  &self:SystemSpawnScript.SystemScriptSource.SystemScriptGraph.NiagaraNodeAssignment_2.SetVariables_#1
  &SystemState
  &InterpolateOverTime
 ]
}
emitters {
 MainTrail {
  emitterUpdateScript [
   &EmitterState(Life Cycle Mode=::NewEnumerator1, Loop Duration=1, UseLoopDelay=false)
   &FN_SpawnAnimTrail(Number of Anim Trails=2)
  ]
  particleSpawnScript [
   &FN_InitializeAnimTrailBoneIndecies(First Bone or Socket Index=1, Second Bone or Socket Index=0)
   &FN_InitializeAnimTrailBoneIndecies(AnimTrail Selection=1, First Bone or Socket Index=3, Second Bone or Socket Index=2) {module=FN_InitializeAnimTrailBoneIndecies001}
   &FN_InitializeAnimTrail(
    AnimTrailWidth=&FN_FrontEndFloatModifier(Gameplay=1, Front End=1.1)
    Color=&SelectLinearColorFromArray(
     Array Sampling Mode=::NewEnumerator1
     Direct Array Index=&MakeIntFromFloat(Float=$User.VariantColor)
     Color Selection Array=DataInterface() {propertyValues {ColorData [<rgba 0 0 0 1>, <rgba 1 1 1 0.5>, <rgba 0 0 0 1>]}}
     Random Seed=?
    )
    Position Offset=&FN_FrontEndVector3Modifier(
     Gameplay=<0 -20 0>
     Front End=&SelectVectorFromArray(Direct Array Index=$Particles.SpawnGroup, Vector Selection Array=DataInterface() {propertyValues {FloatData [<50 20 -8>, <-50 20 -8>]}})
    )
   )
   &AddVelocity(Velocity=&MakeCustomVectorFromBool(
     Bool=$User.bIsFrontEndPreview
     True Vector=<0 -1500 0>
     False Vector=&Lerp_Vector(
      A=&Lerp_Vector(A=<0 300 0>, B=<0 450 0>, Alpha=$System.InterpolateOverTime001.InterpolateOverTime_Float.Moving Average)
      B=<0 1000 0>
      Alpha=$System.InterpolateOverTime.InterpolateOverTime_Float.Moving Average
     )
    ))
  ]
  particleUpdateScript [
   &FN_UpdateAnimTrail()
   &ScaleRibbonWidth(Ribbon Width Scale=&MakeCustomFloatFromBool(
     Bool=$User.bIsFrontEndPreview
     True Float=1
     False Float=&Lerp_Float(
      A=&Lerp_Float(A=0.8, B=0.9, Alpha=$System.InterpolateOverTime001.InterpolateOverTime_Float.Moving Average)
      B=1
      Alpha=$System.InterpolateOverTime.InterpolateOverTime_Float.Moving Average
     )
    ))
   &ScaleColor(
    Scale Alpha=&FloatFromCurve(
     FloatCurve=DataInterface() {propertyValues {Curve {keys [<at 0 1>,<at 0.5 1>,<at 1 0>]}}}
     Scale Curve=&FN_FrontEndFloatModifier(Gameplay=0.5, Front End=0.5)
    )
    ScaleA=true
    ScaleRGB=true
   )
   &SolveForcesAndVelocity(Clamp Velocity=false, Limit Acceleration=false)
   &FN_Sprite_Attributes(
    Color=&FNTimeOfDayColor(
     Night Color=&SelectLinearColorFromArray(
      Array Sampling Mode=::NewEnumerator1
      Direct Array Index=&MakeIntFromFloat(Float=$User.VariantColor)
      Color Selection Array=DataInterface() {propertyValues {ColorData [<rgba 0.2 0.1 0 1.5>, <rgba 0.06 0.06 0.2 1.5>, <rgba 0.12 0.2 0.1467 1.5>]}}
      Random Seed=?
     )
     Morning Color=&SelectLinearColorFromArray(
      Array Sampling Mode=::NewEnumerator1
      Direct Array Index=&MakeIntFromFloat(Float=$User.VariantColor)
      Color Selection Array=DataInterface() {propertyValues {ColorData [<rgba 0.2 0.1 0 1.5>, <rgba 0.1067 0.06 0.2 1.5>, <rgba 0.2 0.13 0.06 1.5>]}}
      Random Seed=?
     )
     Day Color=&SelectLinearColorFromArray(
      Array Sampling Mode=::NewEnumerator1
      Direct Array Index=&MakeIntFromFloat(Float=$User.VariantColor)
      Color Selection Array=DataInterface() {propertyValues {ColorData [<rgba 0.2 0.0733 0.01 1.5>, <rgba 0.0833 0.06 0.2 1.5>, <rgba 0.2 0.2 0.08 1.5>]}}
      Random Seed=?
     )
     Evening Color=&SelectLinearColorFromArray(
      Array Sampling Mode=::NewEnumerator1
      Direct Array Index=&MakeIntFromFloat(Float=$User.VariantColor)
      Color Selection Array=DataInterface() {propertyValues {ColorData [<rgba 0.2 0.1 0 1.5>, <rgba 0.1067 0.06 0.2 1.5>, <rgba 0.2 0.13 0.06 1.5>]}}
      Random Seed=?
     )
    )
    Create Color Attribute=true
    Module Output:ENiagara_AttributeSamplingApplyOutput=::NewEnumerator1
   ) {module=FN_Sprite_Attributes001}
   &FN_Sprite_Attributes(
    Color=&FNTimeOfDayColor(
     Night Color=&SelectLinearColorFromArray(
      Array Sampling Mode=::NewEnumerator1
      Direct Array Index=&MakeIntFromFloat(Float=$User.VariantColor)
      Color Selection Array=DataInterface() {propertyValues {ColorData [<rgba 5 2.875 0.75 1>, <rgba 0 0 6 1>, <rgba 0.35 0.7 0.4667 1>]}}
      Random Seed=?
     )
     Morning Color=&SelectLinearColorFromArray(
      Array Sampling Mode=::NewEnumerator1
      Direct Array Index=&MakeIntFromFloat(Float=$User.VariantColor)
      Color Selection Array=DataInterface() {propertyValues {ColorData [<rgba 6 2 0 1>, <rgba 2 0 6 1>, <rgba 1 0.65 0.3 1>]}}
      Random Seed=?
     )
     Day Color=&SelectLinearColorFromArray(
      Array Sampling Mode=::NewEnumerator1
      Direct Array Index=&MakeIntFromFloat(Float=$User.VariantColor)
      Color Selection Array=DataInterface() {propertyValues {ColorData [<rgba 6 2 0 1>, <rgba 1 0 6 1>, <rgba 1 0.65 0.3 1>]}}
      Random Seed=?
     )
     Evening Color=&SelectLinearColorFromArray(
      Array Sampling Mode=::NewEnumerator1
      Direct Array Index=&MakeIntFromFloat(Float=$User.VariantColor)
      Color Selection Array=DataInterface() {propertyValues {ColorData [<rgba 6 2 0 1>, <rgba 2 0 6 1>, <rgba 1 0.65 0.3 1>]}}
      Random Seed=?
     )
    )
    Create Color Attribute=true
    Module Output:ENiagara_AttributeSamplingApplyOutput=::NewEnumerator1
   )
  ]
  rendererClasses [&NiagaraRibbonRendererProperties]
  renderers [
   {rendererClass=&NiagaraRibbonRendererProperties}
   {rendererClass=&NiagaraRibbonRendererProperties, rendererIndex=1}
   {rendererClass=&NiagaraRibbonRendererProperties, rendererIndex=2}
  ]
 }
 SpeedLine {
  simTarget=GPUComputeSim
  emitterUpdateScript [
   &EmitterState(Life Cycle Mode=::NewEnumerator1, Loop Delay=1, Loop Duration=1, UseLoopDelay=false)
   &SpawnRate(
    SpawnRate=&MakeCustomFloatFromBool(Bool=$User.bIsFrontEndPreview, True Float=200, False Float=150)
    Use Spawn Probability=true
   ) {module=SpawnRate001}
  ]
  particleSpawnScript [
   &InitializeParticle(
    Color=&MakeLinearColorFromVectorAndFloat(
     "Vector (RGB)"=<1 1 1>
     "Float (Alpha)"=&Multiply_Float(
      A=&UniformRangedFloat(Minimum=0.1, Maximum=0.2)
      B=&MakeCustomFloatFromBool(Bool=$User.bIsFrontEndPreview, True Float=0.5, False Float=1)
     )
    )
    Lifetime Max=0.1
    Lifetime Min=0.1
    Lifetime Mode=::NewEnumerator1
    Sprite Size Max=<2 15>
    Sprite Size Min=<2 5>
    Sprite Size Mode=::NewEnumerator2
    UsePositionOffset=false
    Write Material Random=true
   )
   &ShapeLocation(
    Handle Radius=0
    Large Radius=200
    MaskSpawnGroup=false
    Offset=&RandomRangeVector(Minimum=<0 200 0>, Maximum=<0 -300 0>)
    Offset Mode=::NewEnumerator0
    Override Seed=false
    Rotation Axis=<0 1 0>
    Rotation Mode=::NewEnumerator1
    Shape Origin=$Engine.Emitter.SimulationPosition
    Shape Primitive=::NewEnumerator3
    UseOwnerScale=true
   )
   &AddVelocity(Velocity=<0 500 0>, Velocity Speed Scale=&FN_FrontEndFloatModifier(Gameplay=1, Front End=-1)) {module=AddVelocity001}
  ]
  particleUpdateScript [
   &ParticleState()
   &ScaleSpriteSize(Scale Factor=&Vector2DFromCurve(
     Vector2Curve=DataInterface() {propertyValues {XCurve {keys [<at 0 0.5 RCIM_Cubic>,<at 0.5 1 RCIM_Cubic>]}, YCurve {keys [<at 0 0 RCIM_Cubic>,<at 0.5 1 RCIM_Cubic>]}}}
     ScaleCurve=&Vector2DFromFloat(Value=1)
    ))
   &SolveForcesAndVelocity(Clamp Velocity=false, Limit Acceleration=false)
   &ScaleSpriteSizeBySpeed(Max Scale Factor=<1 30>, Min Scale Factor=<0.01 0.05>, Velocity Threshold=1)
   &ScaleColor(
    Scale Alpha=&Multiply_Float(
     A=&FloatFromCurve(
      FloatCurve=DataInterface() {propertyValues {Curve {keys [<at 0 0 RCIM_Cubic>, <at 0.2 1>, {arriveTangent=-2.221, interpMode=RCIM_Cubic, leaveTangent=-2.221, tangentMode=RCTM_User, time=1, value=0}]}}}
      Scale Curve=&MakeCustomFloatFromBool(
       Bool=$User.bIsFrontEndPreview
       True Float=&MakeCustomFloatFromBool(Bool=$User.bIsFullyDeployed, True Float=0.5, False Float=0)
       False Float=&Lerp_Float(A=0, B=1, Alpha=$System.InterpolateOverTime.InterpolateOverTime_Float.Moving Average)
      )
     )
     B=&FloatFromCurve(FloatCurve=DataInterface() {propertyValues {Curve {keys [<at 0 0>,<at 0.3 1>,<at 1 1>]}}}, CurveIndex=$Emitter.Age)
    )
    Scale RGB=&VectorFromFloat(Value=3)
    ScaleA=true
    ScaleRGB=true
   )
  ]
  rendererClasses [&NiagaraSpriteRendererProperties]
  renderers [{rendererClass=&NiagaraSpriteRendererProperties}]
 }
}
summary {
 systemName=NS_Glider_LemonWrath
 emitters [
  {emitterName=SpeedLine, simTarget=GPUComputeSim, rendererClasses [&NiagaraSpriteRendererProperties]}
  {emitterName=MainTrail, rendererClasses [&NiagaraRibbonRendererProperties]}
 ]
 userVariables [
  User.bIsFrontEnd {type=NiagaraBool}
  User.bIsFrontEndPreview {type=NiagaraBool}
  User.bIsFullyDeployed {type=NiagaraBool}
  User.ForwardDot {type=NiagaraFloat, defaultValue=0, description=""}
  User.RibbonIDOffset {type=NiagaraInt32}
  User.VariantColor {type=NiagaraFloat, defaultValue=0, description=""}
 ]
}
